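# Unlit material for Sparks stage screens, showing a storm of slowly rotating tiling clouds with a scrolling pixel layer.
# Two multiplied cloud samples take the stage primary color or, via JamScren?, hue-cycling jam colors.
# Two panned pixel samples, masked by vertical falloff and cell noise, are added on top before MF_Sparks_Screen_Details adds screen detail.
# Audio BeatLifetime modulates cloud brightness, pixel offset and pixel opacity.
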Material /SparksMaps/Effects/Screens/Storm/M_Spark_Screens_Storm
graph {
 properties {
  bAutomaticallySetUsageInEditor=false
  bHasPixelAnimation=true
  bUsedWithNanite=true
  pixelDepthOffsetMode=PDOM_Legacy
  shadingModel=MSM_Unlit
  usedShadingModels=Unlit
 }
 nodes [
  n1 = TextureCoordinate()
  n2 = ScalarParameter() {defaultValue=1, parameterName=ScaleGlobal}
  n3 = Rotator(Coordinate=Add(A=MaterialFunctionCall(UVs=@n1, Texture Scale=@n2) {materialFunction=&ScaleUVsByCenter}.UVs) {constB=0.5}) {speed=-0.03}
  n4 = TextureCoordinate() {uTiling=0.5, vTiling=0.5}
  n5 = Rotator(Coordinate=Add(A=MaterialFunctionCall(UVs=@n4, Texture Scale=@n2) {materialFunction=&ScaleUVsByCenter}.UVs) {constB=0.5}) {speed=-0.05}
  n6 = MaterialFunctionCall!(CustomUPosition) {materialFunction=/SparksStageShow/Systems/StageManager/FXManager/MaterialFunctions/MF_MojoStageColors}
  n7 = MaterialFunctionCall() {materialFunction=&LinearGradient}
  n8 = ScalarParameter() {defaultValue=1, parameterName=Pow}
  n9 = LinearInterpolate(A=Power(Base=@n7.UGradient, Exp=@n8), B=Power(Base=@n7.VGradient, Exp=@n8), Alpha=ScalarParameter() {parameterName=ColorGradientAngle})
  n10 = LinearInterpolate(A=OneMinus(None=@n9), B=@n9, Alpha=ScalarParameter() {parameterName=FlipGradient})
  n11 = NamedRerouteUsage()
  n12 = NamedRerouteUsage()
  n13 = Multiply(
   A=Multiply(
    A=TextureSample(UVs=@n3) {texture=/SparksEffects/Effects/Environment/S7/Textures/T_TilingClouds_D}.RGB
    B=TextureSample(UVs=@n5) {texture=&T_TilingClouds_D}.RGB
   )
   B=LinearInterpolate(
    A=LinearInterpolate(A=@n6.Primary, B=@n6.Primary, Alpha=@n10)
    B=LinearInterpolate(A=@n11, B=@n12, Alpha=@n10)
    Alpha=ScalarParameter() {parameterName=JamScren?}
   )
  )
  n14 = Multiply(A=@n13, B=ScalarParameter() {defaultValue=0.4, parameterName=MeasureMult})
  n15 = CollectionParameter() {collection=/FMAudioReactiveData/AudioAnalysis/MPC_Sparks_Audio_Parameters, parameterName=BeatLifetime}
  MaterialFunctionCall!(Emissive=Add(A=NamedRerouteUsage(), B=LinearInterpolate(A=@n13, B=@n14, Alpha=@n15)), Scanlines Tile) {materialFunction=/SparksMaps/Effects/Screens/Materials/MF/MF_Sparks_Screen_Details}
  n16 = Multiply(A=Time(), B=Constant() {desc=ColorChangeSpeed, r=0.01})
  NamedRerouteDeclaration(None=MaterialFunctionCall(Hue Shift Percentage=Sine(None=@n16), Texture=Constant3Vector() {constant=<rgba 1 0 0.3789 1>}) {materialFunction=&HueShift}.Result) {name=JamColorPrimary, nodeColor=<rgba 0.4297 1 0 1>}
  n17 = OneMinus(None=ComponentMask(None=TextureCoordinate()) {g=true})
  n18 = ScalarParameter() {defaultValue=7, parameterName=PixelMaskHeight}
  n19 = ScalarParameter() {defaultValue=1, parameterName=PixelWidth}
  n20 = Constant() {r=1}
  n21 = ScalarParameter() {defaultValue=1, parameterName=PixelScale}
  n22 = MaterialFunctionCall(UVs=Multiply(A=AppendVector(A=@n19, B=@n20), B=TextureCoordinate() {vTiling=-4}), Texture Scale=@n21) {materialFunction=&ScaleUVsByCenter}
  n23 = LinearInterpolate(
   A=ScalarParameter() {parameterName=PixelBeatLow}
   B=ScalarParameter() {defaultValue=0.01, parameterName=PixelBeatHigh}
   Alpha=CollectionParameter() {collection=&MPC_Sparks_Audio_Parameters, parameterName=BeatLifetime}
  )
  n24 = AppendVector(
   A=ScalarParameter() {parameterName=Pixel_1_Tile_Horz}
   B=Add(A=ScalarParameter() {defaultValue=0.1, parameterName=Pixel_1_Tile_Vert}, B=@n23)
  )
  n25 = TextureSample(UVs=Panner(Coordinate=Add(A=@n22.UVs, B=@n24)) {speedX=0.05}) {samplerSource=SSM_Wrap_WorldGroupSettings, texture=&/T_Sparsk_Screen_Storm_Pixels}
  n26 = Multiply(A=@n23) {constB=0.75}
  n27 = ScalarParameter() {defaultValue=0.1, parameterName=Pixel_2_Tile_Vert}
  n28 = Add(
   A=MaterialFunctionCall(UVs=TextureCoordinate() {vTiling=4}, Texture Scale=@n21) {materialFunction=&ScaleUVsByCenter}.UVs
   B=AppendVector(A=ScalarParameter() {defaultValue=0.22, parameterName=Pixel_2_Tile_Horz}, B=Add(A=@n26, B=@n27))
  )
  n29 = Multiply(A=TextureSample(UVs=Panner(Coordinate=@n28) {speedX=0.025}) {samplerSource=SSM_Wrap_WorldGroupSettings, texture=&T_Sparsk_Screen_Storm_Pixels}.R) {constB=0.25}
  n30 = TextureSample(UVs=Panner(Coordinate=TextureCoordinate() {uTiling=0.5, vTiling=0.5}) {speedX=0.1, speedY=-0.1}) {samplerType=SAMPLERTYPE_Grayscale, texture=/Game/Weapons/FORT_Melee/Pickaxe_Cube_Ninja_Male/FX/T_CellNoise}
  n31 = ScalarParameter() {defaultValue=10, parameterName=CloudMask}
  n32 = Multiply(A=Step(Y=Power(Base=@n17, Exp=@n18) {constExponent=7}) {constX=0.0001}, B=Multiply(A=Add(A=@n25.R, B=@n29), B=Multiply(A=@n30.R, B=@n31)))
  n33 = LinearInterpolate(
   B=ScalarParameter() {parameterName=PiexelOpacity}
   Alpha=CollectionParameter() {collection=&MPC_Sparks_Audio_Parameters, parameterName=BeatLifetime}
  ) {constA=1}
  n34 = MaterialFunctionCall() {materialFunction=&MF_MojoStageColors}
  n35 = NamedRerouteUsage()
  n36 = ScalarParameter() {parameterName=JamScren?}
  NamedRerouteDeclaration(None=Multiply(A=Multiply(A=@n32, B=@n33), B=LinearInterpolate(A=@n34.Primary, B=@n35, Alpha=@n36))) {name=Pixels, nodeColor=<rgba 1 0.0703 0 1>}
  n37 = Multiply(A=Time(), B=Constant() {desc=ColorChangeSpeed, r=0.015})
  NamedRerouteDeclaration(None=MaterialFunctionCall(Hue Shift Percentage=Sine(None=@n37), Texture=Constant3Vector() {constant=<rgba 0 0.4829 1 1>}) {materialFunction=&HueShift}.Result) {name=JamColorSecondary, nodeColor=<rgba 1 0 0.75 1>}
  n38 = ScalarParameter() {parameterName=GradientSub}
  n39 = ScalarParameter() {defaultValue=1, parameterName=GradientDiv}
  Saturate(None=Divide(A=Subtract(B=@n38), B=@n39))
  Saturate(None=Divide(A=Subtract(B=@n38), B=@n39))
 ]
}
